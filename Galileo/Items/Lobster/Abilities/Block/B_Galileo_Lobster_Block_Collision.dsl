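; Block collision actor for the Galileo Lobster (lightsaber) whose building health serves as the wielder's poise while blocking.
; Health changes update weapon poise; server damage restarts delayed heal-effect regen, triggers deflect cues and sends blocked-damage events.
; Lobster hits knock the blocker back, while a cooldown-gated path can spawn projectiles back at attackers.
; Death from damage applies block break; EndPlay resets weapon poise to work around a replication race.
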(event EventBeginPlay
  (bind _startinghealth (Variables|Floats|GetStartingHealth))
  (bind _returnvalue (ScalableFloat|GetValueatLevel _startinghealth))
  ;[0 0.6696 0 0.767] Begin Play
  (|CallParentFunction)
  (Gameplay|ForceBuildingHealth (ScalableFloat|GetValueatLevel (Variables|Floats|GetStartingHealth)))
  (bind _asfort_player_pawn (Utilities|Casting|CastToFortPlayerPawn (Actor|GetOwner)))
  (bind _output_get (Variables|Default|SetPlayerPawn _asfort_player_pawn))
  (CallFunction|RemovalChecks _output_get)
  (CallFunction|SetTeam (Team|GetActorTeam _output_get))
  (Health|AssignOnBuildingHealthChanged (EventDispatchers|CreateEvent :!event Custom|OnBuildingHealthChanged_Event))
  ;[1 1 1] Regen Poise after Block Break
  (Networking|CanExecuteCosmeticEvents
    (:False
      (bind _returnvalue_2 (GameplayEffects|GetActiveEffectswithAllTags (Ability|GetAbilitySystemComponent (Variables|Default|GetPlayerPawn)) (Variables|Tags|GetT_BlockBreakDuration)))
      (if (Utilities|IsArrayNonEmpty _returnvalue_2)
        (bind _returnvalue_3 (Ability|GameplayEffect|GetActiveGameplayEffectRemainingDuration (Utilities|Array|Get(acopy) _returnvalue_2)))
        (Health|SetHealth (Math|Float|MapRangeClamped _returnvalue_3 0.0 5.0 _returnvalue 1.0))
        (Development|PrintString (Utilities|String|ToString(Float) (Math|Float|MapRangeClamped _returnvalue_3 0.0 5.0 _returnvalue 1.0)) false false)
        (GameplayEffects|ApplyGameplayEffectToSelf (Ability|GetAbilitySystemComponent self) (Variables|Default|GetGE_Heal) 0.0 (Utilities|Struct|MakeGameplayEffectContextHandle))))))

(event Custom|OnBuildingHealthChanged_Event
  (bind _bp_galileolobster (Variables|Default|GetBP_GalileoLobster))
  ;[0.3029 0.3432 1] OnHealthChanged -> Update Poise
  (bind _output_get (Variables|Floats|SetCurrentPoise (Math|Float|NormalizetoRange (Health|GetHealth) 0.0 (Health|GetMaxHealth))))
  (Development|PrintString (Utilities|String|ToString(Float) _output_get) false false)
  (Utilities|IsValid _bp_galileolobster
    (:Is%20Valid
      (Class|BAthenaGalileoLobsterParent|CurrentPoiseChanged _bp_galileolobster _output_get))
    (:Is%20Not%20Valid
      _))
  (if (>= _output_get 1.0)
    (GameplayEffects|RemoveActiveGameplayEffectbySourceEffect (Ability|GetFortAbilitySystemComponent self) (Variables|Default|GetGE_Heal) (Ability|GetFortAbilitySystemComponent self) -1)))

(event Custom|RemovalChecks
  (params (PlayerPawn :type Fort%20Player%20Pawn%20Object%20Reference))
  ;[1 1 1] RemovalDelegates
  (EventDispatchers|BindEventtoOnDied PlayerPawn (delegate Custom|CustomEvent_0))
  (EventDispatchers|AssignOnPlayerStartDBNO PlayerPawn (delegate Custom|OnPlayerStartDBNO_Event_0)))

(event Custom|OnPlayerStartDBNO_Event_0
  ;[1 1 1] RemovalDelegates
  (Development|PrintString "Destroy Block Collision" false false)
  (Variables|Bools|SetOwnerDied true)
  (Gameplay|SilentDie false))

(event Custom|CustomEvent_0
  (params (DamagedActor :type Actor%20Object%20Reference) (Damage :type Float%20%28single-precision%29) (InstigatedBy :type Controller%20Object%20Reference) (DamageCauser :type Actor%20Object%20Reference) (HitLocation :type Vector) (FHitComponent :type Primitive%20Component%20Object%20Reference) (BoneName :type Name) (Momentum :type Vector))
  ;[1 1 1] RemovalDelegates
  (Development|PrintString "Destroy Block Collision" false false)
  (Variables|Bools|SetOwnerDied true)
  (Gameplay|SilentDie false))

(event Custom|StartHealingTimer
  ;[1 1 1] Stop healing if we're damaged.  Wait a little while, and start healing again.
  (GameplayEffects|RemoveActiveGameplayEffectbySourceEffect (Ability|GetFortAbilitySystemComponent self) (Variables|Default|GetGE_Heal) (Ability|GetFortAbilitySystemComponent self) -1)
  (Utilities|Time|ClearandInvalidateTimerbyHandle (Variables|Default|GetRepairTimerHandle))
  (bind _returnvalue (Utilities|Time|SetTimerbyEvent (delegate Custom|ApplyHeal) (ScalableFloat|GetValueatLevel (Variables|Floats|GetPoiseRegenCooldown))))
  (Variables|Default|SetRepairTimerHandle _returnvalue))

(event Custom|ApplyHeal
  ;[1 1 1] Stop healing if we're damaged.  Wait a little while, and start healing again.
  (GameplayEffects|ApplyGameplayEffectToSelf (Ability|GetFortAbilitySystemComponent self) (Variables|Default|GetGE_Heal) 0.0 (Utilities|Struct|MakeGameplayEffectContextHandle)))

(event Custom|RemoveDeflectReaction
  (bind _playerpawn (Variables|Default|GetPlayerPawn))
  ;[1 1 1] Remove the Deflect Reaction GC and force it through for the anim blueprint (so the deflect anim stops
  (GameplayEffects|RemoveActiveGameplayEffect (Ability|GetAbilitySystemComponent _playerpawn) (Variables|Default|GetDeflectReactionHandle) -1)
  (Ability|RemoveGameplayCueReplicated (Variables|Tags|GetT_DeflectReaction))
  (Networking|ForceNetUpdate _playerpawn) ; Doing this because we're using a GC tag to run the anim.\nThis forces the GC through quickly for the other clients to see.
)

(event EventOnDamageServer
  (params (Damage :type Float%20%28single-precision%29) (DamageTags :type Gameplay%20Tag%20Container%20Structure%20%28by%20ref%29) (Momentum :type Vector :default "0, 0, 0") (HitInfo :type Hit%20Result%20Structure%20%28by%20ref%29) (InstigatedBy :type Controller%20Object%20Reference) (DamageCauser :type Actor%20Object%20Reference) (EffectContext :type Gameplay%20Effect%20Context%20Handle%20Structure))
  ;[1 1 1] OnDamageServer - Notify ability. If we receive any damage that has not been filtered out, destroy this actor
  (Utilities|FlowControl|Sequence
    (:then_0
      (if (GameplayTags|HasAnyTags DamageTags (Variables|Default|GetTagsToBreakBlock) true)
        (Variables|Bools|SetBlockBreak true) ; Certain damage tags can instantly break the block, regardless of our health
      ))
    (:then_1
      (|CallParentFunction Damage DamageTags Momentum HitInfo InstigatedBy DamageCauser EffectContext)
      (CallFunction|SendBlockedDamageEvent Damage DamageCauser)
      (Utilities|FlowControl|Sequence
        (:then_0
          (if (Variables|Bools|GetBlockBreak)
            (Actor|DestroyActor)
            (else
              (CallFunction|StartHealingTimer))))
        (:then_1
          (Utilities|DoOncePerFrame
            (:%20%20
              (bind _wasgalileobun (CallFunction|ExecuteBlockGameplayCue Damage DamageCauser EffectContext InstigatedBy HitInfo (Math|Conversions|ToVector(LinearColor) (Variables|/Galileo/Items/Lobster/Actors/B_Athena_Galileo_Lobster_Parent.B_Athena_Galileo_Lobster_Parent_C|GetBlade%20Color%20Hot (Variables|Default|GetBP_GalileoLobster))) (Math|Conversions|ToVector(LinearColor) (Variables|/Galileo/Items/Lobster/Actors/B_Athena_Galileo_Lobster_Parent.B_Athena_Galileo_Lobster_Parent_C|GetBlade%20Color%20Cool (Variables|Default|GetBP_GalileoLobster)))))
              (if (and _wasgalileobun (Variables|Bools|GetCanReflect))
                (Variables|Bools|SetCanReflect false)
                (Variables|Default|SetReflectTarget (Utilities|Casting|CastToFortPlayerPawn (Pawn|GetControlledPawn InstigatedBy)))
                (CallFunction|SpawnReflectProj HitInfo _wasgalileobun)
                (Utilities|Time|SetTimerbyEvent (delegate Custom|SetCanReflect) (ScalableFloat|GetValueatLevel (Variables|Floats|GetReflectProjectileCooldown)))))) ; Limit Block GC's to one per frame
        )
        (:then_2
          (if (<= (Health|GetHealth) 0)
            (Gameplay|SilentDie false)))))))

(event Custom|SetCanReflect
  ;[1 1 1] OnDamageServer - Notify ability. If we receive any damage that has not been filtered out, destroy this actor
  (Variables|Bools|SetCanReflect true))

(event GameplayCue|GameplayCue.Damage
  (params (EventType :type EGameplayCueEvent%20Enum :default "OnActive") (Parameters :type Gameplay%20Cue%20Parameters%20Structure%20%28by%20ref%29))
  (bind (_normalizedmagnitude _rawmagnitude _effectcontext _matchedtagname _originaltag _aggregatedsourcetags _aggregatedtargettags _location _normal _instigator _effectcauser) (Ability|GameplayCue|BreakGameplayCueParameters Parameters))
  (bind (_bblockinghit _binitialoverlap _time _distance _location_2 _impactpoint) (Collision|BreakHitResult (Ability|EffectContext|GetHitResult _effectcontext)))
  (bind _playerpawn (Variables|Default|GetPlayerPawn))
  ;[1 1 1] Proto Impact FX logic
  (switch Utilities|FlowControl|Switch|SwitchonEFortTeamAffiliation (Team|GetTeamAffiliation (Actor|GetOwner) _effectcauser)
    (:Neutral
      _)
    (:Hostile
      _))
  (Class|BAthenaGalileoLobsterParent|DoBulletBlockEffect (Variables|Default|GetBP_GalileoLobster) _impactpoint)
  (Networking|CanExecuteCosmeticEvents
    (:True
      (if (Pawn|IsLocallyControlled _playerpawn)
        (ForceFeedback|SpawnForceFeedbackatLocation :ForceFeedbackEffect "/Game/Weapons/ForceFeedback/FF_Generic_Impact_Near.FF_Generic_Impact_Near" :Location (Transformation|GetActorLocation _playerpawn) :bLooping false :IntensityMultiplier (select (and (!= (Variables|Floats|GetCurrentPoise) 0.0) (< (Variables|Floats|GetCurrentPoise) 0.2)) (Math|Float|MapRangeClamped (Variables|Floats|GetCurrentPoise) 1.0 0.1 1.0 1e+01) 0.05) :StartTime 0.0 :bAutoDestroy true) ; Quickly ramp up the rumble intensity just before Poise is broken
      ))))

(event Custom|SpawnReflectProj
  (params (HitInfo :type Hit%20Result%20Structure) (WasGalileoBun :type Boolean))
  (bind (_bblockinghit _binitialoverlap _time _distance _location _impactpoint _normal _impactnormal _physmat _hitactor _hitcomponent) (Collision|BreakHitResult HitInfo))
  (bind _reflecttarget (Variables|Default|GetReflectTarget))
  ;[1 1 1] Reflect ranged weapon hits
  (Utilities|IsValid _reflecttarget
    (:Is%20Valid
      (Variables|Bools|SetHitCenter (not (or (Utilities|Equal(Object) _hitcomponent (Variables|Default|GetBox2)) (Utilities|Equal(Object) _hitcomponent (Variables|Default|GetBox3)))))
      (bind _returnvalue (Game|SpawnActorfromClass "/Galileo/Items/Lobster/Abilities/Block/B_Prj_Galileo_Lobster_Reflect.B_Prj_Galileo_Lobster_Reflect_C" (Math|Conversions|ToTransform(Vector) (+ _location (* (Transformation|GetActorForwardVector) 15.0))) "AlwaysSpawn" "MultiplyWithRoot" (Variables|Default|GetPlayerPawn) WasGalileoBun (Variables|Default|GetPlayerPawn)))
      (Variables|/Script/Engine.MovementComponent|SetVelocity (* (Math|Random|RandomUnitVectorinConeinDegrees (- (Transformation|GetActorLocation _reflecttarget) (Transformation|GetActorLocation)) (select (Variables|Bools|GetHitCenter) (ScalableFloat|GetValueatLevel (Variables|Floats|GetCenterReflectConeAngle)) (ScalableFloat|GetValueatLevel (Variables|Floats|GetSideReflectProjectileConeAngle)))) (ScalableFloat|GetValueatLevel (Variables|Floats|GetReflectProjectileSpeed))) (Variables|/Script/FortniteGame.FortProjectileBase|GetProjectileMovementComponent _returnvalue)))))

(event EventOnDeathServer
  (params (Damage :type Float%20%28single-precision%29) (DamageTags :type Gameplay%20Tag%20Container%20Structure%20%28by%20ref%29) (Momentum :type Vector :default "0, 0, 0") (HitInfo :type Hit%20Result%20Structure%20%28by%20ref%29) (InstigatedBy :type Controller%20Object%20Reference) (DamageCauser :type Actor%20Object%20Reference) (EffectContext :type Gameplay%20Effect%20Context%20Handle%20Structure))
  (bind _bp_galileolobster (Variables|Default|GetBP_GalileoLobster))
  (bind _bp_galileolobster_2 (Variables|Default|GetBP_GalileoLobster))
  ;[1 0.5 0] Activate Block Break ability if this actor was destroyed via direct damage.
  (if (not (Variables|Bools|GetOwnerDied))
    (GameplayEffects|ApplyGameplayEffectToSelf :self (Ability|GetFortAbilitySystemComponent (Variables|Default|GetPlayerPawn)) :GameplayEffectClass (Variables|Default|GetGE_BlockBreak) :Level 0.0)
    (Ability|ExecuteGameplayCueWithParamsReplicated (Variables|GameplayCues|GetGC_BlockBreak) (Ability|GameplayCue|MakeGameplayCueParameters :NormalizedMagnitude 1.0 :RawMagnitude 1.0 :EffectContext EffectContext :Location (Math|Conversions|ToVector(LinearColor) (Variables|/Galileo/Items/Lobster/Actors/B_Athena_Galileo_Lobster_Parent.B_Athena_Galileo_Lobster_Parent_C|GetBlade%20Color%20Hot _bp_galileolobster_2)) :Normal (Math|Conversions|ToVector(LinearColor) (Variables|/Galileo/Items/Lobster/Actors/B_Athena_Galileo_Lobster_Parent.B_Athena_Galileo_Lobster_Parent_C|GetBlade%20Color%20Cool _bp_galileolobster_2)) :Instigator InstigatedBy :EffectCauser DamageCauser :SourceObject (Variables|Default|GetBP_GalileoLobster) :GameplayEffectLevel 1 :AbilityLevel 1 :bReplicateLocationWhenUsingMinimalRepProxy false)) ; Moved the Block Break GC from the montage to here so that it's able to be changing colors by variables
    (Class|BAthenaGalileoLobsterParent|BlockBreak _bp_galileolobster_2)
    (Utilities|IsValid _bp_galileolobster
      (:Is%20Valid
        (Class|BAthenaGalileoLobsterParent|DeactivateBlockBreakWarning _bp_galileolobster)))))

(event EventEndPlay
  (params (EndPlayReason :type EEndPlayReason%20Enum :default "Destroyed"))
  ;[1 0.5 0 1 14] FORT-891814
  ; The OnDeath of this actor calls block break functionality and sends a multicast to update the weapon actors poise to manually reset the unstable light saber state to 1. But, there is a race condition where the multicast will execute on a client before the OnBuildingHealthChanged event.
  ; This results in the state being set to 1 by the multicast then OnBuildingHealthChanged setting it to a very low value. A proper fix would probably require a refactor of the poise setup.
  ; -EndPlay seems to consistently execute for all clients after the multicast and healthchange event.
  ; -Getting hit with damage that has tags to specifically break block will automatically call death and avoid changing health which may lead to the same race condition.
  (|CallParentFunction EndPlayReason)
  (bind _bp_galileolobster (Variables|Default|GetBP_GalileoLobster)
    (:then
      (Class|BAthenaGalileoLobsterParent|CurrentPoiseChanged _bp_galileolobster 1.0))
    (:else)))

(fn BlockedLobsterImpulse
  (params (HitInfo :type Hit%20Result%20Structure) (Instigator :type Character%20Object%20Reference))
  (bind (_bblockinghit _binitialoverlap _time _distance _location _impactpoint _normal _impactnormal _physmat _hitactor _hitcomponent _hitbonename _bonename _hititem _elementindex _faceindex _tracestart _traceend) (Collision|BreakHitResult HitInfo))
  (Utilities|IsValid Instigator
    (:Is%20Valid
      _)
    (:Is%20Not%20Valid
      _))
  (Variables|Floats|SetImpulseMultiplier 1.0)
  (Pawn|Character|LaunchCharacterJump (Variables|Default|GetPlayerPawn) (Math|Vector|vector+vector (* (* (Math|Vector|Normalize (- _traceend _tracestart)) (ScalableFloat|GetValueatLevel (Variables|Floats|GetImpulseXY))) (Variables|Floats|GetImpulseMultiplier)) (Math|Vector|MakeVector 0.0 0.0 (ScalableFloat|GetValueatLevel (Variables|Floats|GetImpulseZ)))) false true false true))

(fn ExecuteBlockGameplayCue
  (params (Damage :type Float%20%28double-precision%29) (DamageCauser :type Actor%20Object%20Reference) (EffectContext :type Gameplay%20Effect%20Context%20Handle%20Structure) (Instigator :type Actor%20Object%20Reference) (HitResult :type Hit%20Result%20Structure) (Color%20Hot :type Vector) (Color%20Cool :type Vector))
  (bind _bp_galileolobster (Variables|Default|GetBP_GalileoLobster))
  (bind _currentpoise (Variables|Floats|GetCurrentPoise))
  (bind _returnvalue (Ability|GameplayCue|MakeGameplayCueParameters :NormalizedMagnitude _currentpoise :RawMagnitude Damage :EffectContext EffectContext :Location Color%20Hot :Normal Color%20Cool :Instigator Instigator :EffectCauser DamageCauser :SourceObject _bp_galileolobster :GameplayEffectLevel 1 :AbilityLevel 1 :bReplicateLocationWhenUsingMinimalRepProxy false))
  (bind _playerpawn (Variables|Default|GetPlayerPawn))
  (switch Utilities|FlowControl|Switch|SwitchonEFortTeamAffiliation (Team|GetTeamAffiliation _playerpawn DamageCauser)
    (:Neutral
      _)
    (:Hostile
      _))
  ;[1 1 1] :-(  Force the GameplayCues through so the deflect anims play correctly in the animBP
  (Utilities|Time|ClearandInvalidateTimerbyHandle (Variables|Default|GetDeflectReactionTimerHandle))
  (bind _returnvalue_2 (GameplayEffects|ApplyGameplayEffectToSelf (Ability|GetAbilitySystemComponent _playerpawn) (Variables|Default|GetGE_Deflect) 0.0 EffectContext))
  (Variables|Default|SetDeflectReactionHandle _returnvalue_2)
  (Networking|ForceNetUpdate _playerpawn) ; Doing this because we're using a GC tag to run the anim.\nThis forces the GC through quickly for the other clients to see.
  (bind _returnvalue_3 (Utilities|Time|SetTimerbyEvent (EventDispatchers|CreateEvent :!event RemoveDeflectReaction) 0.08))
  (Variables|Default|SetDeflectReactionTimerHandle _returnvalue_3)
  (bind _asfort_weapon (Utilities|Casting|CastToFortWeapon DamageCauser)
    (:then
      (if (GameplayTags|HasAnyTags (Tags|GetWeaponDataGameplayTags _asfort_weapon) (GameplayTags|MakeLiteralGameplayTagContainer "(GameplayTags=((TagName=\"Item.Weapon.Galileo.Lobster\")))"))
        (Ability|ExecuteGameplayCueWithParamsReplicated (Variables|GameplayCues|GetGC_DeflectLobster) _returnvalue)
        (Development|PrintString "Blocked Damage! Lobster" false false "(R=1.000000,G=0.042986,B=0.000000,A=1.000000)")
        ;[1 1 1] Trigger Knockback on Lobster Impacts
        (CallFunction|BlockedLobsterImpulse HitResult (Utilities|Casting|CastToCharacter (Actor|GetOwner DamageCauser)))
        (return false)
        (else
          _)))
    (:CastFailed
      _))
  (Ability|ExecuteGameplayCueWithParamsReplicated (Variables|GameplayCues|GetGC_DeflectGeneric) _returnvalue)
  (Development|PrintString "Blocked Damage! Generic" false false)
  (return false))

(fn OnRep_CollisionEnabled
  (if (Math|Conversions|ToBoolean(Integer) (Variables|Default|GetCollisionEnabled))
    (Collision|SetActorEnableCollision true)
    (Development|PrintString "LOBSTER BLOCK COLLISION ENABLED" false false "(R=0.165732,G=1.000000,B=0.000000,A=1.000000)")
    ;[1 1 1] Make sure the currently equipped Lobster is cached
    (bind _asb_athena_galileo_lobster_parent (Utilities|Casting|CastToB_Athena_Galileo_Lobster_Parent (Variables|/Script/FortniteGame.FortPawn|GetCurrentWeapon (Variables|Default|GetPlayerPawn))))
    (Variables|Default|SetBP_GalileoLobster _asb_athena_galileo_lobster_parent)
    (else
      (Collision|SetActorEnableCollision false)
      (Development|PrintString "LOBSTER BLOCK COLLISION DISABLED" false false "(R=1.000000,G=0.000000,B=0.038396,A=1.000000)"))))

(fn SendBlockedDamageEvent
  (params (Damage :type Float%20%28double-precision%29) (DamageCauser :type Actor%20Object%20Reference))
  (bind _playerpawn (Variables|Default|GetPlayerPawn))
  ;[1 1 1] For Challenges tracking damage blocked by the Lobster
  (Utilities|IsValid _playerpawn
    (:Is%20Valid
      (Ability|SendGameplayEventtoActor _playerpawn (Variables|Tags|GetT_BlockedDamageEvent) (Utilities|Struct|MakeGameplayEventData (Variables|Tags|GetT_BlockedDamageEvent) _playerpawn DamageCauser 0 0 "()" "(GameplayTags=)" "(GameplayTags=)" Damage)))))

(fn ConstructionScript)
