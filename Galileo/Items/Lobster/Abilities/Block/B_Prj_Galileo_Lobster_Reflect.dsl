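; Projectile for the Galileo Lobster block's reflect ability, which applies a reflect-damage gameplay effect to damageable targets with an ability system.
; On spawn it uses either Galileo Bun or generic tracer visuals and feeds muzzle origin, start time and a speed-mapped duration into the tracer's WPO material.
; On stop it disables collision, gravity and bullet whip, plays pawn or surface impact sound, adds non-mobile surface decals and hides the tracer.
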(event EventBeginPlay
  (bind _tracermesh (Variables|Default|GetTracerMesh))
  (bind _tracermesh_2 (Variables|Default|GetTracerMesh))
  (bind _p_galileo_bun_projectile_glow (Variables|Default|GetP_Galileo_Bun_Projectile_Glow))
  (|CallParentFunction)
  (Networking|CanExecuteCosmeticEvents
    (:True
      ;[1 1 1] Projectile visuals (use generic particles and mesh if not a Galileo Bun)
      (if (not (Variables|Default|GetUseGalileoBunTracer))
        (Effects|Components|ParticleSystem|SetTemplate _p_galileo_bun_projectile_glow (Variables|Default|GetGenericTracer))
        (Transformation|SetRelativeRotation _p_galileo_bun_projectile_glow "0, 0, 0" false false)
        (Transformation|SetRelativeScale3D _p_galileo_bun_projectile_glow "0.500000,0.500000,0.500000")
        (Components|StaticMesh|SetStaticMesh _tracermesh (Variables|Default|GetGenericMesh))
        (Rendering|Material|CreateDynamicMaterialInstance _tracermesh 0 "/Game/Effects/Fort_Effects/Materials/M_Projectile_Bullet_WPO.M_Projectile_Bullet_WPO" "None" :!class /Script/Engine.PrimitiveComponent))
      (bind _asfort_player_pawn (Utilities|Casting|CastToFortPlayerPawn (Variables|Actor|GetInstigator)))
      (Rendering|Material|SetVectorParameterValueonMaterials _tracermesh_2 "BulletOrigin" (Effects|GetMuzzleLocation (Variables|/Script/FortniteGame.FortPawn|GetCurrentWeapon _asfort_player_pawn) -1))
      (Rendering|Material|SetScalarParameterValueonMaterials _tracermesh_2 "StartTime" (Utilities|Time|GetGameTimeinSeconds))
      (Rendering|Material|SetScalarParameterValueonMaterials _tracermesh_2 "WPODuration" (Math|Float|MapRangeClamped (Math|Vector|VectorLength (Transformation|GetVelocity)) 5e+03 4e+04 0.1 0.25)))))

(event EventOnStop
  (params (Hit :type Hit%20Result%20Structure%20%28by%20ref%29))
  (bind (_bblockinghit _binitialoverlap _time _distance _location _impactpoint _normal _impactnormal _physmat _hitactor) (Collision|BreakHitResult Hit))
  (bind _returnvalue (Ability|GetFortAbilitySystemComponent _hitactor))
  (bind (_asfort_pawn _bsuccess) (Utilities|Casting|CastToFortPawn _hitactor))
  (Collision|SetCollisionEnabled (Variables|Collision|GetCapsuleComponent) "NoCollision")
  (Utilities|IsValid _returnvalue
    (:Is%20Valid
      ;[1 1 1] Deal Damage if we hit something that can take damage.
      (if (Variables|/Script/Engine.Actor|GetbCanBeDamaged _hitactor)
        (GameplayEffects|ApplyGameplayEffectToTarget :self (Ability|GetFortAbilitySystemComponent (Variables|Actor|GetInstigator)) :GameplayEffectClass "/Galileo/Items/Lobster/Abilities/Block/GE_Galileo_Lobster_Reflect_Damage.GE_Galileo_Lobster_Reflect_Damage_C" :Target _returnvalue :Level 0.0)))
    (:Is%20Not%20Valid
      _))
  (Utilities|FlowControl|Sequence
    (:then_0
      (Variables|/Script/Engine.ProjectileMovementComponent|SetProjectileGravityScale 0.0 (Variables|Movement|GetProjectileMovementComponent))
      (Class|BulletWhipTrackerComponentBase|Disable (Variables|Visual|GetBulletWhipTrackerComponent)))
    (:then_1
      (Networking|CanExecuteCosmeticEvents
        (:True
          (Audio|PlaySoundatLocation (select _bsuccess (Variables|Default|GetSniper_Rifle_PlayerImpactSound) (Variables|Default|GetSniper_Rifle_SurfaceImpactSound)) (Transformation|GetActorLocation))
          (if (not (or (Fortnite|IsMobilePlatform) _bsuccess))
            (bind _returnvalue_2 (Rendering|Decal|SpawnDecalatLocation "/Game/Weapons/Materials/MD_Decal_BulletHole_Lifetime_Inst1.MD_Decal_BulletHole_Lifetime_Inst1" (Variables|Default|GetDecalSize) _location (Math|Conversions|RotationFromXVector (* _impactnormal -1.0)) 4.5))
            (Rendering|Components|Decal|SetFadeOut _returnvalue_2 0.0 4.5 true)
            (Utilities|FlowControl|Delay 0.2))
          (Rendering|SetVisibility (Variables|Default|GetTracerMesh) false false))))))

(fn ConstructionScript
  (Projectile|SetGravityScale (ScalableFloat|GetValueatLevel (Variables|Default|GetProjectileGravityScale))))
