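; Gameplay ability for blocking with the Galileo Lobster weapon, held until input release, subject to a minimum duration unless hit.
; Activation applies a blocking effect and, where cosmetic events cannot execute, reuses or spawns an attached collision actor, disabled on end.
; Tag listeners cancel it for conflicting actions like swimming or jetpacking, while grounded jump presses send a dodge event.
; Forced net updates speed replication of the GameplayCue-tag-driven animation.
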(event Ability|EventActivateAbility
  (bind _playerpawn (Variables|Default|GetPlayerPawn))
  ;[1 1 1] OnActivateAbility -> Init
  (if (CallFunction|CheckJetpackTag)
    (Ability|EndAbility)
    (else
      (Default|InitAbility)
      (|CallParentFunction)
      (Utilities|FlowControl|Sequence
        (:then_0
          ;[1 1 1] This shameful mess listens for GameplayTags that should end the Block ability if they are added while Blocking
          (Utilities|FlowControl|Sequence
            (:then_0
              (Ability|Tasks|WaitGameplayTagAdd "(TagName=\"Gameplay.Action.Player.SurfaceSwimming\")" _playerpawn
                (:Added
                  _)) ; End the ability when player starts swimming
            )
            (:then_1
              (Ability|Tasks|WaitGameplayTagAdd "(TagName=\"Gameplay.Action.Player.CarryingPlayer\")" _playerpawn
                (:Added
                  _)) ; End the ability when player starts carrying a DBNO player
            )
            (:then_2
              (Ability|Tasks|WaitGameplayTagAdd "(TagName=\"Gameplay.Action.Player.InteractSearch\")" _playerpawn
                (:Added
                  _)) ; End the ability when player starts Interact Searching
            )
            (:then_3
              (Ability|Tasks|WaitGameplayTagAdd "(TagName=\"Gameplay.Action.Player.InteractUse\")" _playerpawn
                (:Added
                  _)) ; End the ability when player starts Interact Using
            )
            (:then_4
              (Ability|Tasks|WaitGameplayTagAdd "(TagName=\"Gameplay.Status.HidingProp.GettingInProp\")" _playerpawn
                (:Added
                  _)) ; End the ability when player enters a Hiding Prop
            )
            (:then_5
              (Ability|Tasks|WaitGameplayTagAdd "(TagName=\"MovementMode.Skydiving\")" _playerpawn
                (:Added
                  _)) ; End the ability when player starts Skydiving
            )
            (:then_6
              (Ability|Tasks|WaitGameplayTagAdd "(TagName=\"MovementMode.Parachuting\")" _playerpawn
                (:Added
                  _)) ; End the ability when player starts Skydiving
            )
            (:then_7
              (Ability|Tasks|WaitGameplayTagAdd (Variables|Default|GetSlidingTag)))
            (:then_8
              (Ability|Tasks|WaitGameplayTagAdd "(TagName=\"Event.IsUsingJetpack\")"
                (:Added
                  _)))
            (:then_9
              (Ability|Tasks|WaitGameplayTagAdd "(TagName=\"Gameplay.Action.Player.Ascender\")" _playerpawn
                (:Added
                  _)))
            (:then_10
              (Ability|Tasks|WaitGameplayEvent "(TagName=\"Abilities.Generic.Athena.Ascender.Enter\")" _playerpawn
                (:EventReceived
                  (Variables|Bools|SetBlockInputReleased true))))
            (:then_11
              (Ability|Tasks|WaitGameplayTagAdd "(TagName=\"Granted.Athena.EnvItem.HidingProp.GettingInProp\")" _playerpawn
                (:Added
                  _)))
            (:then_12
              (Ability|Tasks|WaitGameplayTagAdd "(TagName=\"Granted.Athena.EnvItem.HidingProp.HidingInProp\")" _playerpawn
                (:Added
                  _)))
            (:then_13
              (Ability|Tasks|WaitGameplayTagAdd "(TagName=\"Granted.Athena.EnvItem.HidingProp.ExitingProp\")" _playerpawn
                (:Added
                  _)))
            (:then_14
              (Ability|Tasks|WaitGameplayTagAdd "(TagName=\"Granted.Athena.EnvItem.HidingProp.Teleporter\")" _playerpawn
                (:Added
                  _))))
          (Ability|EndAbility))
        (:then_1
          ;[1 1 1] Enable the Blocking Collision and play the Block Start
          (CallFunction|EnableBlockingCollision)
          (Animation|Montage|MontageStopGroupbyName (Components|SkeletalMesh|GetAnimInstance (Variables|/Script/Engine.Character|GetMesh (Variables|Default|GetPlayerPawn))) 0.2 "DefaultGroup") ; Stop monrages so anim BP can do the block start
          (Networking|ForceNetUpdate (Variables|Default|GetPlayerPawn)) ; Doing this because we're using a GC tag to run the anim.\nThis forces the GC through quickly for the other clients to see.
        )
        (:then_2
          (Development|PrintString "waiting for jump")
          ;[1 1 1] Trigger the Dodge Abiility when Jump is pressed
          (Ability|Tasks|WaitGameplayEvent "(TagName=\"Event.Jump.Pressed\")"
            (:EventReceived
              (if (not (AI|Components|NavMovement|IsFalling (Variables|/Script/Engine.Character|GetCharacterMovement (Variables|Default|GetPlayerPawn))))
                (Ability|SendGameplayEvent (Variables|Default|GetT_DodgeEvent))))))))))

(event Ability|EventOnEndAbility
  (params (bWasCancelled :type Boolean))
  ;[1 1 1] OnEndAbility -> Cleanup
  (|CallParentFunction bWasCancelled)
  (CallFunction|EndAbilityCleanup)
  (Development|PrintString "BLOCK ENDED"))

(event Ability|EventOnAbilityInputReleased
  (bind _blockstarttime (Variables|Floats|GetBlockStartTime))
  (bind _returnvalue (Utilities|Time|GetGameTimeinSeconds))
  (bind _returnvalue_2 (- _returnvalue _blockstarttime))
  ;[1 1 1] OnAbilityInputReleased -> End the Ability after the min blocking time
  (|CallParentFunction)
  (Variables|Bools|SetBlockInputReleased true)
  (if (or (Variables|Bools|GetWasBlockCollisionHit) (> _returnvalue_2 (Variables|Floats|GetMinBlockDuration)))
    (Ability|EndAbility)
    (else
      (bind _returnvalue_3 (Utilities|Time|SetTimerbyEvent (EventDispatchers|CreateEvent :!event K2_EndAbility) (- (Variables|Floats|GetMinBlockDuration) _returnvalue_2)))
      (Variables|Default|SetEndAbilityTimerHandle _returnvalue_3))))

(fn CheckJetpackTag
  (return (GameplayTags|HasAnyMatchingGameplayTags (CallFunction|GetAvatarasFortPlayerPawn) (GameplayTags|MakeGameplayTagContainerfromTag "(TagName=\"Event.IsUsingJetpack\")"))))

(fn EnableBlockingCollision
  (bind _playerpawn (Variables|Default|GetPlayerPawn))
  (Networking|CanExecuteCosmeticEvents
    (:False
      (Variables|Bools|SetHasBlockCollision false)
      ;[1 1 1] Do we already have a Blocking Collision Actor Attached?
      (for _array_element (Actor|GetAttachedActors _playerpawn true false)
        (Utilities|IsValid _array_element
          (:Is%20Valid
            (bind _asb_galileo_lobster_block_collision (Utilities|Casting|CastToB_Galileo_Lobster_Block_Collision _array_element))
            (Variables|Bools|SetHasBlockCollision true)
            (Variables|Default|SetBP_BlockCollisionActor _asb_galileo_lobster_block_collision))))
      (if (Variables|Bools|GetHasBlockCollision)
        ;[1 1 1] Turn collision on if we have a blocking collision actor
        (Variables|/Galileo/Items/Lobster/Abilities/Block/B_Galileo_Lobster_Block_Collision.B_Galileo_Lobster_Block_Collision_C|SetCollisionEnabled 1 (Variables|Default|GetBP_BlockCollisionActor))
        (else
          ;[1 1 1] Spawn Blocking Collision Actor if none were found
          (bind _returnvalue (Game|SpawnActorfromClass (Variables|Default|GetActorClass_BlockCollision) (Math|Transform|MakeTransform (Math|Vector|vector+vector (Transformation|GetActorLocation _playerpawn) (Math|Vector|MakeVector 0.0 0.0 (Variables|Floats|GetBlockCollisionActorAttachZOffset)) (* (Transformation|GetActorForwardVector _playerpawn) (Variables|Floats|GetBlockCollisionActorAttachXOffset))) (Transformation|GetActorRotation _playerpawn)) "AlwaysSpawn" "SelectDefaultAtRuntime" _playerpawn _playerpawn))
          (bind _asb_galileo_lobster_block_collision_2 (Utilities|Casting|CastToB_Galileo_Lobster_Block_Collision _returnvalue))
          (bind _output_get (Variables|Default|SetBP_BlockCollisionActor _asb_galileo_lobster_block_collision_2))
          (bind _asb_athena_galileo_lobster_parent (Utilities|Casting|CastToB_Athena_Galileo_Lobster_Parent (Ability|GetCurrentSourceWeapon)))
          (Variables|/Galileo/Items/Lobster/Abilities/Block/B_Galileo_Lobster_Block_Collision.B_Galileo_Lobster_Block_Collision_C|SetBP_GalileoLobster _asb_athena_galileo_lobster_parent _output_get)
          (Transformation|AttachActorToActor _output_get (Variables|Default|GetPlayerPawn) "" "KeepWorld" "KeepWorld" "KeepRelative" true))))))

(fn EndAbilityCleanup
  (bind _playerpawn (Variables|Default|GetPlayerPawn))
  (bind _bp_lobsterweapon (Variables|Default|GetBP_LobsterWeapon))
  (bind _bp_blockcollisionactor (Variables|Default|GetBP_BlockCollisionActor))
  (Development|PrintString "End Block" false false "(R=1.000000,G=0.466819,B=0.000000,A=1.000000)")
  (Utilities|Time|ClearandInvalidateTimerbyHandle (Variables|Default|GetEndAbilityTimerHandle))
  (Ability|RemoveGameplayEffectFromOwnerWithHandle (Variables|Default|GetBlockingEffectHandle))
  (Animation|AnimTrailsNotify (Variables|Default|GetPlayerPawn) false)
  (Utilities|IsValid _bp_blockcollisionactor
    (:Is%20Valid
      (Variables|/Galileo/Items/Lobster/Abilities/Block/B_Galileo_Lobster_Block_Collision.B_Galileo_Lobster_Block_Collision_C|SetCollisionEnabled 0 _bp_blockcollisionactor))
    (:Is%20Not%20Valid
      _))
  (Utilities|IsValid _bp_lobsterweapon
    (:Is%20Valid
      (Class|BAthenaGalileoLobsterParent|ToggleDynamicAudio _bp_lobsterweapon true))
    (:Is%20Not%20Valid
      _))
  (Utilities|IsValid _playerpawn
    (:Is%20Valid
      (Networking|ForceNetUpdate _playerpawn) ; Doing this because we're using a GC tag to run the anim.\nThis forces the GC through quickly for the other clients to see.
    )))

(fn InitAbility
  (bind _returnvalue (Ability|CommitAbility))
  (if _returnvalue
    (Development|PrintString "Start Block" true false "(R=0.000000,G=0.401284,B=1.000000,A=1.000000)")
    (bind _asfort_player_pawn (Utilities|Casting|CastToFortPlayerPawn (Ability|GetAvatarActorfromActorInfo))
      (:then
        (bind _output_get (Variables|Default|SetPlayerPawn _asfort_player_pawn))
        (Character|UnCrouch _output_get) ; Force player to stop crouching
        (bind _asb_athena_galileo_lobster_parent (Utilities|Casting|CastToB_Athena_Galileo_Lobster_Parent (Ability|GetCurrentSourceWeapon))
          (:then
            (bind _output_get_2 (Variables|Default|SetBP_LobsterWeapon _asb_athena_galileo_lobster_parent))
            (Class|BAthenaGalileoLobsterParent|EventonBlockActivate _output_get_2)
            (Class|BAthenaGalileoLobsterParent|ToggleDynamicAudio _output_get_2 false)
            (bind _returnvalue_2 (Ability|ApplyGameplayEffectToOwner (Variables|Default|GetGE_Blocking)))
            (Variables|Default|SetBlockingEffectHandle _returnvalue_2)
            (Variables|Floats|SetBlockStartTime (Utilities|Time|GetGameTimeinSeconds))
            (Variables|Bools|SetWasBlockCollisionHit false)
            (Variables|Bools|SetBlockInputReleased false))
          (:CastFailed
            _)))
      (:CastFailed
        _))
    (Ability|EndAbility)
    (else
      _)))
